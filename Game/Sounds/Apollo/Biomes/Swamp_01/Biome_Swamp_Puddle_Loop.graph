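# Sound cue for a looping swamp puddle ambience in the Apollo biome set, routed to the outdoor ambient sound class at reduced volume.
# A single looping wave player feeds a modulator that varies pitch while holding volume fixed.
# Cue-level attenuation overrides give it a short-range falloff with distance low-pass filtering and stereo spread, alongside a concurrency override.
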SoundCue /Game/Sounds/Apollo/Biomes/Swamp_01/Biome_Swamp_Puddle_Loop
root=Modulator_1
nodes [
 {id=Modulator_1, children [WavePlayer_0], properties {pitchMin=0.9, volumeMax=1, volumeMin=1}}
 {id=WavePlayer_0, properties {bLooping=true, soundWaveAssetPtr=&/Biome_Swamp_Puddle_Loop_01}}
]
properties {
 bOverrideAttenuation=true
 bOverrideConcurrency=true
 maxDistance=1100
 soundClassObject=&Fort_Ambient_Outdoors_Athena
 volumeMultiplier=0.2
 attenuationOverrides {bAttenuateWithLPF=true, falloffDistance=900, lPFRadiusMin=500, stereoSpread=300, attenuationShapeExtents=<200 90>}
}
